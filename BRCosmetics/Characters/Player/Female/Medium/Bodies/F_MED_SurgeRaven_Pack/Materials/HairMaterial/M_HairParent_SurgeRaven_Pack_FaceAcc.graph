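# Parent character material for the SurgeRaven Pack FaceAcc hair, two-sided, with the shading model chosen by expression.
# The shared CharacterParent PreFX attributes feed an FN hair function, blended in by its HairMask behind the UseFNHair2_0 switch.
# An optional HasFlyAwayHair? switch adds a flyaway hair function, which a UV1 sphere mask excludes near the UV origin.
# The result passes into the shared CharacterParent PostFX function.
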Material /BRCosmetics/Characters/Player/Female/Medium/Bodies/F_MED_SurgeRaven_Pack/Materials/HairMaterial/M_HairParent_SurgeRaven_Pack_FaceAcc
graph {
 properties {
  bUseMaterialAttributes=true
  bUsedWithClothing=true
  bUsedWithInstancedStaticMeshes=true
  bUsedWithMeshDeformer=true
  bUsedWithMorphTargets=true
  bUsedWithNanite=true
  bUsedWithSkeletalMesh=true
  pixelDepthOffsetMode=PDOM_Legacy
  shadingModel=MSM_FromMaterialExpression
  twoSided=true
  usedShadingModels=Default Lit | Subsurface | Clear Coat | Subsurface Profile | Cloth
 }
 nodes [
  n1 = MaterialFunctionCall() {materialFunction=&MF_CharacterParent_2023_PreFX}
  n2 = MaterialFunctionCall!(In=@n1.Result, Mask Input=@n1.Result) {materialFunction=&/MF_FN_Hair_SurgeRaven_Pack_FaceAcc}
  n3 = BlendMaterialAttributes(
   A=@n1.Result
   B=StaticSwitchParameter(True=@n2.Result, False=@n1.Result) {defaultValue=true, group=Skin FX, parameterName=UseFNHair2_0}
   Alpha=@n2.HairMask
  )
  n4 = MaterialFunctionCall!(MaterialInput=@n3) {materialFunction=&/MF_FlyAwayHair_SurgeRaven_Pack_FaceAcc}
  n5 = SphereMask(A=TextureCoordinate() {coordinateIndex=1}, B=Constant()) {attenuationRadius=0.01}
  MaterialFunctionCall(
   Material=StaticSwitchParameter(True=BlendMaterialAttributes(A=@n4.Result, B=@n3, Alpha=@n5), False=@n3) {group=Skin FX, parameterName=HasFlyAwayHair?}
   M=@n1.M
   Pre Skinned Local Position=@n1.M
   Distance Blend=@n1.M
   Fresnel=@n1.M
  ) {materialFunction=&MF_CharacterParent_2023_PostFX}
  Reroute(None=@n1.Tech Art Mask)
 ]
}
